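; Lockdown gate wall that closes to block navigation and delay players when a lockdown begins.
; The construction script records the wall component's world location as the start and a point a fixed height above it as the end.
; Gate_OnLockdownBegin plays a normalized timeline, rate-scaled by LockdownTime, to lerp the wall between those points.
; BeginPlay, overlap and Tick events are empty; a comment notes the lockdown event is called from a test-level cheat pedestal.
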(event EventBeginPlay)

(event Collision|EventActorBeginOverlap
  (params (OtherActor :type Actor%20Object%20Reference)))

(event EventTick
  (params (DeltaSeconds :type Float%20%28single-precision%29)))

(event Custom|Gate_OnLockdownBegin
  ;[0.2292 0.1224 0.01746] Gate_OnLockdownBegin
  ; Current called from the cheat pedistal in the test level
  ; Closes the gate to block navigation and delay players
  (Components|Timeline|SetPlayRate (Variables|BA_Dropout_Wall|GetTimeline_OpenTrapDoor) (Math|Float|SafeDivide 1.0 (Variables|Default|GetLockdownTime))) ; set the play rate so it will scale \nthe normalized timeline to the \ndesired time
  ;;[0 0.5926 1] Animate the door slamming closed with a simple normalized timeline and lerp
  (bind (_direction _height_normalized) (Timeline|Timeline_OpenTrapDoor Play)
    (:Update
      (Transformation|SetWorldLocation (Variables|Default|GetS_SetName_Wall_01) (Math|Vector|Lerp(Vector) (Variables|Default|GetWallStartPos) (Variables|Default|GetWallEndPos) _height_normalized) false false))))

(fn ConstructionScript
  (bind _s_setname_wall_01 (Variables|Default|GetS_SetName_Wall_01))
  (bind _returnvalue (Transformation|GetWorldLocation _s_setname_wall_01))
  (Variables|Default|SetWallStartPos _returnvalue)
  (Variables|Default|SetWallEndPos (Math|Vector|MakeVector (.x _returnvalue) (.y _returnvalue) (+ (.z _returnvalue) 384.0))))
