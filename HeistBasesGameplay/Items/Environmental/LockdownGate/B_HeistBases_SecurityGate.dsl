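; Heist base security gate Blueprint that slams shut after a delay on lockdown to block navigation and delay players.
; The server wakes dormancy, triggers collision and navigation blocking, activates an ability for overlaps, then returns to dormant.
; A multicast drives a timeline scaling the visual mesh, with music, sound, dust and shake; OnRep corrects out-of-relevancy clients.
; Hotfix floats control close time, explosive vulnerability and instant self-removal; gates can ignore lockdown.
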(event EventBeginPlay
  (bind _self self)
  (bind _returnvalue (Ability|GetAbilitySystemComponent _self))
  ;[0 1 0.05694] BeginPlay
  ; Apply the damage vulnerability to explosives (GE) if enabled
  (|CallParentFunction)
  (Networking|SwitchHasAuthority
    (:Authority
      ;;[1 0.3923 0] Heist Base features are disabled for some playlists and should be set to a "vanilla" state - for Security Gates, that is destroying themselves
      (if (ScalableFloat|GetValueasBool (Variables|Hotfix|GetHF_RemoveInstanceImmediately))
        (Gameplay|SilentDie false)
        (elif (ScalableFloat|GetValueasBool (Variables|Hotfix|GetHF_ExplosiveVulnerabilityEnabled))
          (Utilities|IsValid _returnvalue
            (:Is%20Valid
              (GameplayEffects|ApplyGameplayEffectToSelf :self _returnvalue :GameplayEffectClass "/Game/Athena/Items/Parents/Damage/GE_Damage_Vulnerability_Explosive.GE_Damage_Vulnerability_Explosive_C" :Level 0.0))))))))

(event Custom|OnLockdownBegin
  ;[0.2292 0.1224 0.01746] Gate_OnLockdownBegin
  ; Current called from the cheat pedistal in the test level
  ; Closes the gate to block navigation and delay players
  (if (not (Variables|Gate|Lockdown|GetbIgnoreLockdown))
    (if (not (Variables|Gate|GetbGateStartsClosed))
      (Utilities|FlowControl|Delay (Variables|Gate|GetLockdownDelay))
      (Variables|/HeistBasesGameplay/Audio/Data/Heist_Music_Component.Heist_Music_Component_C|SetbIsLockdown true (Variables|Default|GetHeist_Music_Component))
      (CallFunction|RequestCloseGate))))

(event Custom|RequestCloseGate
  ;[0.3698 0.8086 1] RequestCloseGate
  ; Will only close the gate if it is not already closed. Gates that start closed or closed by a laser are ignored
  (if (not (Variables|Gate|Closing|GetbIsGateClosed))
    (CallFunction|CloseGate)))

(event Custom|CloseGate
  ;[0.2292 0.1224 0.01746] CloseGate [PRIVATE] - should only call "RequestCloseGate" externally
  ; Animate the Gate closed, Handle additional collision cases: impulse objects out, destroy objects, destroy gate
  (Networking|SwitchHasAuthority
    (:Authority
      (Utilities|FlowControl|Sequence
        (:then_0
          ;;[0 1 0.02679] INIT: setup the gate to be closed
          (Networking|SetNetDormancy "DORM_Awake") ; set dormancy AWAKE so door animates on the client
          (Variables|Default|SetCachedHF_GateCloseTime (ScalableFloat|GetValueatLevel (Variables|Hotfix|GetHF_GateCloseTime))) ; cache the scalable float to the client
        )
        (:then_1
          ;;[0.5198 0.8052 1] CLOSE GATE: Set collision, call close gate visuals, Knockback player
          (CallFunction|MultiCloseGate) ; replicated for relevant clients
          (CallFunction|OnCheckGateCloseCollision)
          (CallFunction|SetGateCollisionClosed))
        (:then_2
          ;;[1 0.4304 0] COMPLETED: once animation is complete, set dormancy back to dormant for perf and set gate closed for clients outside relevancy
          (Utilities|FlowControl|Delay (Variables|Default|GetCachedHF_GateCloseTime)) ; Caution with delays on a sequence node, \nlater sequences do not wait for the delay async
          (Networking|SetNetDormancy "DORM_DormantAll"))
        (:then_3
          (Fortnite|AI|SetComponentAffectNavigationGeneration (Variables|Default|GetNavCollision_MainGate) true))))))

(event Custom|MultiCloseGate
  ;[0.3058 0 1] MultiCloseGate
  ; Muticast for relevant clients- animate the gate closed and hancle cosmetic VFX/SFX
  (Networking|CanExecuteCosmeticEvents
    (:True
      (Variables|Gate|Closing|SetbIsGateClosing true)
      (Utilities|FlowControl|Sequence
        (:then_0
          ;;[0 0.5926 1] Animate the door slamming closed with a simple normalized timeline and lerp
          (Components|Timeline|SetPlayRate (Variables|B_HeistBases_SecurityGate|GetTimeline_CloseLockdownGate) (Math|Float|SafeDivide 1.0 (Variables|Default|GetCachedHF_GateCloseTime))) ; set the play rate so it will scale \nthe normalized timeline to the \ndesired time
          (bind (_direction _height_normalized) (Timeline|Timeline_CloseLockdownGate Play)
            (:Update
              (CallFunction|UpdateGateClosing _height_normalized))
            (:Finished
              ;;[1 0.4528 0] Set the Gate Closed and reset closing bool for OnRep updates
              (Variables|Gate|Closing|SetbIsGateClosing false)
              (CallFunction|SetGateCollisionClosed))
            (:InitialImpact
              ;;;[0.3058 0 1] Timeline Event: when the door first touches the ground for any FX
              (Networking|CanExecuteCosmeticEvents
                (:True
                  (CallFunction|OnDoorVisuallySlammed))) ; only cosmetic effects should be called here
            )))
        (:then_1
          ;;[0.3058 0 1] Audio
          (Audio|SpawnSoundAttached (Variables|FX|Sound|GetGateCloseSound) (Variables|Default|GetLockdown_Gate_visuals))
          (Class|HeistMusicComponent|RequestMusicPlayback (Variables|Default|GetHeist_Music_Component) "NewEnumerator1"))))))

(event Custom|UpdateGateClosing
  (params (CurrentClosingPercent :type Float%20%28double-precision%29))
  ;[0 0.5926 1] UpdateGateClosing
  ; Animate the door slamming closed with a simple normalized timeline and lerp
  (Transformation|SetRelativeScale3D (Variables|Default|GetLockdown_Gate_visuals) (Math|Vector|MakeVector 1.0 1.0 (Math|Float|Lerp (.z (Variables|Gate|Closing|GetGateOpenScale)) 1.0 CurrentClosingPercent))))

(event Custom|OnCheckGateCloseCollision
  (bind _self self)
  (bind _returnvalue (Ability|GetAbilitySystemComponent _self))
  ;[0.3297 0.4902 1] OnCheckGateCloseCollision
  ; Use a GA to target the area and determine how to handle any overlaps on the closing gate
  (Utilities|IsValid _returnvalue
    (:Is%20Valid
      (Ability|Advanced|GrantandActivateAbilityonSelf _returnvalue (Variables|Gate|Closing|GetGateClosingAbility) _self))))

(event Custom|OnDoorVisuallySlammed
  (bind _self self)
  (bind _returnvalue (Transformation|GetActorLocation _self))
  ;[0.3793 0 1] OnDoorSlammed
  ; (Replicated to Clients)
  ; Play VFX dust burst to reinforce the door slamming
  (Networking|CanExecuteCosmeticEvents
    (:True
      (Effects|Components|ParticleSystem|SpawnFXAtLocation "/HeistBasesGameplay/FX/NS_Heist_SecureDoorSlam.NS_Heist_SecureDoorSlam" (+ (Transformation|GetWorldLocation (Variables|Default|GetWall_WalkwayFrame)) "0.000000,0.000000,5.000000") (Transformation|GetWorldRotation (Variables|Default|GetWall_WalkwayFrame)))
      ;;[0.3793 0 1] Camera shake and force feedback for gate slamming shut
      (Camera|PlayWorldCameraShake "/Game/Effects/Fort_Effects/BluePrints/CameraShakes/B_Medium_Explosion_CameraShake.B_Medium_Explosion_CameraShake_C" _returnvalue 0.0 (Variables|FX|GetFX_CameraShakeRadius))
      (ForceFeedback|SpawnForceFeedbackatLocation "/Game/Weapons/ForceFeedback/Explosions/FF_Generic_Explosion_Near.FF_Generic_Explosion_Near" _returnvalue))))

(fn MarkIgnoreLockdown
  ;[0.01785 0.02242 0.07812] MarkIgnoreLockdown: Sets the gate to ignore lockdown.
  ; Primarily called from B_HeistBaseLaserGroup.
  (Networking|SwitchHasAuthority
    (:Authority
      (Variables|Gate|Lockdown|SetbIgnoreLockdown true)
      (Fortnite|FortShippingLog self "MarkIgnoreLockdown: Invoked."))))

(fn OnLockdownInitiated
  (CallFunction|OnLockdownBegin)
  (return true))

(fn OnRep_bIsGateClosed
  (Networking|CanExecuteCosmeticEvents
    (:True
      (if (not (Variables|Gate|Closing|GetbIsGateClosing))
        ;[0.4219 0 1] Replicate gate status for remote clients out of relevancy when gate was closed
        (Transformation|SetRelativeScale3D (Variables|Default|GetLockdown_Gate_visuals) (Math|Vector|SelectVector "1.000000,1.000000,1.000000" (Variables|Gate|Closing|GetGateOpenScale) (Variables|Gate|Closing|GetbIsGateClosed)))))))

(fn SetGateCollisionClosed
  (Transformation|SetRelativeScale3D (Variables|Default|GetLockdown_Gate_collision) "1.000000,1.000000,1.000000") ; will be called on client+server to keep collision in sync
  (Networking|SwitchHasAuthority
    (:Authority
      (Variables|Gate|Closing|SetbIsGateClosed true))))

(fn ConstructionScript
  (Variables|Gate|Closing|SetbIsGateClosed (Variables|Gate|GetbGateStartsClosed)) ; sets visual and collision for gate
)
